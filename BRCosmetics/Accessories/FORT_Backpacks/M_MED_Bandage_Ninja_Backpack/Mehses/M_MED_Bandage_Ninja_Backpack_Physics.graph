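# Physics asset for the Bandage Ninja backpack cosmetic, defining simple collision and joint limits for its sword and two bow bones.
# Four capsule bodies cover the backpack root, a long sword and two small bow pieces.
# The sword is constrained to the root, and both bows branch from the sword.
# All joints lock twist and allow only narrow swing, with the sword limited to a single swing axis.
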PhysicsAsset /BRCosmetics/Accessories/FORT_Backpacks/M_MED_Bandage_Ninja_Backpack/Mehses/M_MED_Bandage_Ninja_Backpack_Physics
bodies {
 backpack_root [{
   shapeType=Capsule
   length=35.0292
   radius=9.1893
   shapeName=backpack_root_capsule
   center=<-9.04 -7.5885 5.436>
   rotation=<rot 0 -90.1472 -80.7882>
  }]
 dyn_bow_01 [{shapeType=Capsule, length=2.2298, radius=1.3638, shapeName=dyn_bow_01_capsule, center=<5 0 -5>, rotation=<rot 90 0 0>}]
 dyn_bow_02 [{shapeType=Capsule, length=1.9431, radius=1.7019, shapeName=dyn_bow_02_capsule, center=<1 0 4>, rotation=<rot 90 0 0>}]
 dyn_sword [{shapeType=Capsule, length=68.7503, radius=2.9596, shapeName=dyn_sword_capsule, center=<4 0 0>, rotation=<rot 90 -87.6318 -83.8457>}]
}
constraints [
 {bone1Name=dyn_bow_01, bone2Name=dyn_sword, swing1LimitDegrees=10, swing2LimitDegrees=10, twistLimitDegrees=0}
 {bone1Name=dyn_bow_02, bone2Name=dyn_sword, swing1LimitDegrees=20, swing2LimitDegrees=10, twistLimitDegrees=0}
 {bone1Name=dyn_sword, bone2Name=backpack_root, swing1LimitDegrees=0, swing2LimitDegrees=10, twistLimitDegrees=0}
]
